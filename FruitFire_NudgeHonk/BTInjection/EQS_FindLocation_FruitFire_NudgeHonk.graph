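# EQS query for the FruitFire NudgeHonk behavior tree injection that supplies a location at the AI's current leash center.
# A Current Location generator takes the point from the custom CurrentLeashCenter context and projects it onto the navmesh, searching up and down with the AvoidObstacles nav filter.
# It has no tests, so no scoring or filtering follows generation.
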EnvQuery /FruitFire_NudgeHonk/BTInjection/EQS_FindLocation_FruitFire_NudgeHonk
options [{
  generator {
   name=EnvQueryGenerator_CurrentLocation_0
   class=EnvQueryGenerator_CurrentLocation
   properties {
    bAutoSortTests=true
    bCanRunAsync=false
    envQueryResultNormalizationOption=Default
    navDataOverrideContext=null
    optionName=""
    queryContext=&/EnvQueryContext_CurrentLeashCenter_FruitFire_NudgeHonk.EnvQueryContext_CurrentLeashCenter_FruitFire_NudgeHonk_C
    projectionData {
     bCanDisableTrace=true
     bCanProjectDown=true
     bCanTraceOnGeometry=true
     bCanTraceOnNavMesh=true
     bOnlyBlockingHits=true
     bTraceComplex=false
     extentX=0
     extentY=10
     extentZ=10
     navigationFilter=/Game/AI/NavFilters/NavFilter_AvoidObstacles.NavFilter_AvoidObstacles_C
     postProjectionVerticalOffset=0
     projectDown=1024
     projectUp=1024
     serializedChannel=ECC_Visibility
     traceChannel=TraceTypeQuery1
     traceMode=Navigation
     traceProfileName=None
     traceShape=Line
     versionNum=1
    }
   }
  }
  tests []
 }]
