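# Injected AI behavior tree for the FruitFire NudgeHonk idle animation, which positions or pauses a bot, then triggers a gameplay ability.
# Alert level picks the first stage: Suspicious or LostTarget runs an EQS query and moves there, while Unaware or Threatened waits, each self-aborting on change.
# Force Success lets the sequence continue to a parallel that runs the tag-keyed ability alongside an Escalate task, gated by the root evaluator service's execution status.
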BehaviorTree /FruitFire_NudgeHonk/BTInjection/BT_AnimIdle_FruitFire_NudgeHonk
blackboard=&/BB_AnimIdle_FruitFire_NudgeHonk
root {
 name=Selector
 class=BTComposite_Selector
 children [Sequence {
   class=BTComposite_Sequence
   children [
    Selector {
     class=BTComposite_Selector
     children [
      Sequence {
       class=BTComposite_Sequence
       children [
        Find Root Idle Location {
         class=BTTask_RunEQSQuery
         properties {
          nodeName=Find Root Idle Location
          blackboardKey {selectedKeyName=AIEvaluator_Patrolling_Destination}
          eQSRequest {queryTemplate=&/EQS_FindLocation_FruitFire_NudgeHonk, runMode=SingleResult}
         }
        }
        Bot Move To {
         class=FortAthenaBTTask_BotMoveTo
         decorators []
         properties {
          movementResultKeyName=None
          nodeName=Bot Move To
          acceptableRadius {defaultValue=5}
          allowRandomWobble {defaultValue=true}
          bAllowStrafe {defaultValue=true}
          blackboardKey {selectedKeyName=AIEvaluator_Patrolling_Destination}
          observedBlackboardValueTolerance {defaultValue=4.75}
         }
        }
       ]
       decorators [Blackboard Based Condition {
         class=BTDecorator_Blackboard
         properties {
          arithmeticOperation=Equal
          flowAbortMode=Self
          intValue=1
          nodeName=Blackboard Based Condition
          notifyObserver=ValueChange
          stringValue=Suspicious
          blackboardKey {selectedKeyName=AIEvaluator_Global_AlertLevel}
         }
        }]
       properties {children [
         {childTask=&self:BTTask_RunEQSQuery_2}
         {childTask=&self:FortAthenaBTTask_BotMoveTo_0}
        ]}
      }
      Sequence {
       class=BTComposite_Sequence
       children [
        Find Root Idle Location {
         class=BTTask_RunEQSQuery
         properties {
          nodeName=Find Root Idle Location
          blackboardKey {selectedKeyName=AIEvaluator_Patrolling_Destination}
          eQSRequest {queryTemplate=&EQS_FindLocation_FruitFire_NudgeHonk, runMode=SingleResult}
         }
        }
        Bot Move To {
         class=FortAthenaBTTask_BotMoveTo
         decorators []
         properties {
          movementResultKeyName=None
          nodeName=Bot Move To
          acceptableRadius {defaultValue=5}
          allowRandomWobble {defaultValue=true}
          bAllowStrafe {defaultValue=true}
          blackboardKey {selectedKeyName=AIEvaluator_Patrolling_Destination}
          observedBlackboardValueTolerance {defaultValue=4.75}
         }
        }
       ]
       decorators [Blackboard Based Condition {
         class=BTDecorator_Blackboard
         properties {
          arithmeticOperation=Equal
          flowAbortMode=Self
          intValue=2
          nodeName=Blackboard Based Condition
          notifyObserver=ValueChange
          stringValue=LostTarget
          blackboardKey {selectedKeyName=AIEvaluator_Global_AlertLevel}
         }
        }]
       properties {children [
         {childTask=&self:BTTask_RunEQSQuery_0}
         {childTask=&self:FortAthenaBTTask_BotMoveTo_1}
        ]}
      }
      Exit BT if no longer unaware {
       class=BTTask_Wait
       decorators [Blackboard Based Condition {
         class=BTDecorator_Blackboard
         properties {
          arithmeticOperation=Equal
          flowAbortMode=Self
          intValue=0
          nodeName=Blackboard Based Condition
          notifyObserver=ValueChange
          stringValue=Unaware
          blackboardKey {selectedKeyName=AIEvaluator_Global_AlertLevel}
         }
        }]
       properties {nodeName=Exit BT if no longer unaware, waitTime {defaultValue=5}}
      }
      Exit BT if no longer unaware {
       class=BTTask_Wait
       decorators [Blackboard Based Condition {
         class=BTDecorator_Blackboard
         properties {
          arithmeticOperation=Equal
          flowAbortMode=Self
          intValue=3
          nodeName=Blackboard Based Condition
          notifyObserver=ValueChange
          stringValue=Threatened
          blackboardKey {selectedKeyName=AIEvaluator_Global_AlertLevel}
         }
        }]
       properties {nodeName=Exit BT if no longer unaware, waitTime {defaultValue=5}}
      }
     ]
     decorators [Force Success {class=BTDecorator_ForceSuccess}]
     properties {
      children [
       {
        childComposite=&self:BTComposite_Sequence_0
        childTask=null
        decorators [&self:BTDecorator_Blackboard_3]
       }
       {
        childComposite=&self:BTComposite_Sequence_2
        childTask=null
        decorators [&self:BTDecorator_Blackboard_0]
       }
       {
        childComposite=null
        childTask=&self:BTTask_Wait_0
        decorators [&self:BTDecorator_Blackboard_4]
       }
       {
        childComposite=null
        childTask=&self:BTTask_Wait_1
        decorators [&self:BTDecorator_Blackboard_1]
       }
      ]
      services []
     }
     services []
    }
    Simple Parallel {
     class=BTComposite_SimpleParallel
     children [
      TriggerAbility {
       class=FortBTTask_ExecuteGameplayAbility
       decorators []
       properties {nodeName=TriggerAbility, gameplayAbilityTagContainer {key=Ability_TagContainer, defaultValue {gameplayTags []}}}
      }
      FortAthenaBTTask_Escalate_0 {
       class=FortAthenaBTTask_Escalate
       children []
       decorators [Escalate? {
         class=BTDecorator_Blackboard
         properties {
          arithmeticOperation=GreaterOrEqual
          flowAbortMode=None
          intValue=3
          nodeName=Escalate?
          stringValue=ExecutionPending
          blackboardKey {selectedKeyName=AIEvaluator_Escalate_ExecutionStatus}
         }
        }]
       properties {bIgnoreRestartSelf=false, nodeName=""}
       services []
      }
     ]
     properties {
      children [
       {childTask=&self:FortBTTask_ExecuteGameplayAbility_1}
       {
        childTask=&self:FortAthenaBTTask_Escalate_0
        decorators [&self:BTDecorator_Blackboard_2]
       }
      ]
      services []
     }
     services []
    }
   ]
   decorators []
   properties {children [
     {
      childComposite=&self:BTComposite_Selector_1
      childTask=null
      decorators [&self:BTDecorator_ForceSuccess_0]
     }
     {childComposite=&self:BTComposite_SimpleParallel_1, childTask=null}
    ]}
  }]
 decorators []
 properties {
  children [{childComposite=&self:BTComposite_Sequence_1, childTask=null, decorators []}]
  services [&self:FortAthenaBTService_AIEvaluator_0]
 }
 services [Evaluating... Escalate {
   class=FortAthenaBTService_AIEvaluator
   properties {interval=0.5, nodeName=Evaluating... Escalate, aIEvaluatorClass {defaultValue=/Script/FortniteAIServer.FortAthenaAIBotEvaluator_Escalate}}
  }]
}
