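; EQS context for the FruitFire NudgeHonk boss that supplies its current leash center as a single location.
; It casts the querier to BP_PlayerPawn_NonParticipant_Boss and reads the patrol path from the patrolling component on that pawn's AI controller.
; The location of the path's first patrol point serves as the center; a failed cast returns the world origin.
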(fn ProvideSingleLocation
  (params (QuerierObject :type Object%20Reference) (QuerierActor :type Actor%20Object%20Reference))
  (bind _asbp_player_pawn_non_participant_boss (Utilities|Casting|CastToBP_PlayerPawn_NonParticipant_Boss QuerierActor)
    (:then
      (return (Transformation|GetActorLocation (Class|FortAthenaPatrolPath|GetPatrolPoint (Class|FortAthenaNpcPatrollingComponent|GetPatrolPath (Actor|GetComponentbyClass (AI|GetAIController _asbp_player_pawn_non_participant_boss) "/Script/FortniteAI.FortAthenaNpcPatrollingComponent") :!class /Script/FortniteAI.FortAthenaNpcPatrollingComponent) 0))))
    (:CastFailed
      (return "0, 0, 0"))))
